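# HUD element widget that shows a bottom-centered "Focus" key prompt during special events, collapsing itself on touch input.
# On Construct it binds to the local player controller's focal-point-used delegate and unbinds on Destruct.
# Using a focal point plays the HidePrompt opacity fade forward, and releasing it reverses the fade.
# The bound action is set in PreConstruct, and a custom event can swap in an Input Action and mapping context.
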WidgetBlueprint /SpecialEventGameplay/UI/SpecialEvent_FocusPrompt
animations [HidePrompt {
  binding_count=1
  length_seconds=0.3
  object=&self:HidePrompt
  sequence {bindings [Overlay_0 {
     id=#1
     possessed_class=Overlay
     tracks [{
       type=MovieSceneFloatTrack
       display_name=Render Opacity
       sections [{type=MovieSceneFloatSection, end_frame=6, start_frame=0, channels [None {type=MovieSceneScriptingFloatChannel, keys [[0,1],[6,0]]}]}]
      }]
    }]}
 }]
graphs [EventGraph {nodes [
   n1 = OnFocalPointUsedChanged!() -> (
    OutputDelegate
    then
    FocalPointUser:Fort Player Controller Athena Object Reference
    FocalPointActor:Actor Object Reference
    bUsingFocalPoint:bool=~
   ) {is=title K2Node_CustomEvent}
   > n2 = Branch(Condition=@n1.bUsingFocalPoint)
   n3 = Get HidePrompt() -> (HidePrompt:Widget Animation Object Reference)
   > PlayAnimationForward(InAnimation=@n3)
   PlayAnimationReverse(execute=@n2.else, InAnimation=@n3)
   n4 = SetBoundInputAction!() -> (OutputDelegate, then, InputAction:Input Action Object Reference, InputMappingContext:Input Mapping Context Object Reference) {is=title K2Node_CustomEvent}
   > SetInputContextOverride!(
    execute
    self:Fort Keybind Widget Object Reference=Get KeyBind_Focus() -> (KeyBind_Focus:Keybind Widget Object Reference)
    NewContextOverride:Input Mapping Context Object Reference=@n4.InputMappingContext
   ) -> (then)
   > SetBoundInputAction!(execute, self:Fort Keybind Widget Object Reference=Get KeyBind_Focus(), NewAction:Input Action Object Reference=@n4.InputAction) -> (then) {is=title K2Node_CallFunction}
   > FortShippingLog!(
    execute
    WorldContextObject:Object Reference
    InString:String=Append(A="Updatting focus key binding in widget to: ", B=GetDisplayName(Object=@n4.InputAction))
    bLogAsWarning:bool=false
   ) -> (then)
   n5 = Event PreConstruct()
   > Parent: PreConstruct(IsDesignTime=@n5.IsDesignTime)
   > SetBoundAction!(execute, self:Fort Keybind Widget Object Reference=Get KeyBind_Focus(), NewBoundAction:Name=Get BoundAction() -> (BoundAction:Name=None)) -> (then)
   Event Construct()
   > Parent: Construct()
   > n6 = Sequence()
   > Branch(Condition=IsUsingTouch(PlayerController=GetOwningPlayer()))
   > SetVisibility(InVisibility=Collapsed)
   n7 = Cast To FortPlayerControllerAthena(
    execute=@n6.then_1
    Object=GetFirstLocalFortPlayerController!(bIncludeNonSpectators:bool=true, bIncludeSpectators:bool=false) -> (ReturnValue:Fort Player Controller Object Reference)
   )
   > n8 = Set BoundController(BoundController:Fort Player Controller Athena Object Reference=@n7) -> (Output_Get:Fort Player Controller Athena Object Reference)
   n9 = Create Event()
   > Bind Event to On Using Focal Point Changed(self:Fort Player Controller Athena Object Reference=@n8, Delegate=@n9) {is=title K2Node_AddDelegate}
   Event Destruct()
   > Parent: Destruct()
   n10 = Get BoundController() -> (BoundController:Fort Player Controller Athena Object Reference)
   > Is Valid(InputObject=@n10)
   > Unbind Event from On Using Focal Point Changed(self:Fort Player Controller Athena Object Reference=@n10, Delegate=@n9) {is=title K2Node_RemoveDelegate}
  ]}]
variables [
 BoundAction {type=string, default=Target}
 BoundController {type=FortPlayerControllerAthena, default=null}
]
widgets {
 info {inheritedWidgetCount=1, parentClass=&FortHUDElementWidget, rootWidgetClass=&Overlay, widgetCount=6}
 widgets [
  Overlay_0 {widgetClassPath=&Overlay, widget=Overlay_0}
  SafeZone_0 {
   widgetClassPath=&SafeZone
   bIsVariable=true
   parent=Overlay_0
   widget=SafeZone_0
   slot {class=OverlaySlot, properties {horizontalAlignment=HAlign_Fill, verticalAlignment=VAlign_Fill}}
  }
  Line_Focus {
   widgetClassPath=&HorizontalBox
   bIsVariable=true
   parent=SafeZone_0
   widget=Line_Focus
   properties {renderTransform {angle=0, scale=<1 1>, shear=<0 0>, translation=<0 -100>}}
   slot {class=SafeZoneSlot, properties {hAlign=HAlign_Center, vAlign=VAlign_Bottom, padding=<ltrb 0 0 0 100>}}
  }
  KeyBind_Focus {
   widgetClassPath=/FortUILibrary/Input/KeybindWidget
   bIsVariable=true
   parent=Line_Focus
   widget=KeyBind_Focus
   properties {bShowUnboundStatus=true, boundAction=EventModeFocusLock}
   slot {class=HorizontalBoxSlot, properties {verticalAlignment=VAlign_Center}}
  }
  Text_Focus {
   widgetClassPath=/Script/UMG.RichTextBlock
   bIsVariable=true
   parent=Line_Focus
   widget=Text_Focus
   properties {
    bOverrideDefaultStyle=true
    text=Focus
    textStyleSet=/Game/Athena/HUD/TextStyles/PlayersLeftStyles
    defaultTextStyleOverride {
     overflowPolicy=Clip
     transformPolicy=None
     colorAndOpacity {colorUseRule=UseColor_Specified, specifiedColor=<rgba 1 1 1 1>}
     font {
      bForceMonospaced=false
      bMaterialIsStencil=false
      fontMaterial=null
      fontObject=&BurbankSmall
      letterSpacing=0
      monospacedWidth=1
      size=18
      skewAmount=0
      typefaceFontName=Bold
      outlineSettings {
       bApplyOutlineToDropShadows=true
       bMiteredCorners=false
       bSeparateFillAlpha=true
       outlineMaterial=&M-FontShadow
       outlineSize=3
       outlineColor=<rgba 0 0 0 0.75>
      }
     }
     highlightColor {colorUseRule=UseColor_Specified, specifiedColor=<rgba 0 0 0 1>}
     highlightShape {
      bIsDynamicallyLoaded=false
      drawAs=Image
      imageType=NoImage
      mirroring=NoMirror
      resourceName=None
      resourceObject=null
      tiling=NoTile
      imageSize=<32 32>
      margin=<ltrb 0 0 0 0>
      outlineSettings {
       bUseBrushTransparency=false
       roundingType=HalfHeightRadius
       width=0
       color {colorUseRule=UseColor_Specified, specifiedColor=<rgba 0 0 0 0>}
       cornerRadii=<0 0 0 0>
      }
      tintColor {colorUseRule=UseColor_Specified, specifiedColor=<rgba 1 1 1 1>}
      uVRegion {bIsValid=false, max=<0 0>, min=<0 0>}
     }
     selectedBackgroundColor {colorUseRule=UseColor_Foreground, specifiedColor=<rgba 1 0 1 0>}
     shadowColorAndOpacity=<rgba 0 0 0 1>
     shadowOffset=<0 0>
     strikeBrush {
      bIsDynamicallyLoaded=false
      drawAs=Image
      imageType=NoImage
      mirroring=NoMirror
      resourceName=None
      resourceObject=null
      tiling=NoTile
      imageSize=<32 32>
      margin=<ltrb 0 0 0 0>
      outlineSettings {
       bUseBrushTransparency=false
       roundingType=HalfHeightRadius
       width=0
       color {colorUseRule=UseColor_Specified, specifiedColor=<rgba 0 0 0 0>}
       cornerRadii=<0 0 0 0>
      }
      tintColor {colorUseRule=UseColor_Specified, specifiedColor=<rgba 1 1 1 1>}
      uVRegion {bIsValid=false, max=<0 0>, min=<0 0>}
     }
     underlineBrush {
      bIsDynamicallyLoaded=false
      drawAs=Image
      imageType=NoImage
      mirroring=NoMirror
      resourceName=None
      resourceObject=null
      tiling=NoTile
      imageSize=<32 32>
      margin=<ltrb 0 0 0 0>
      outlineSettings {
       bUseBrushTransparency=false
       roundingType=HalfHeightRadius
       width=0
       color {colorUseRule=UseColor_Specified, specifiedColor=<rgba 0 0 0 0>}
       cornerRadii=<0 0 0 0>
      }
      tintColor {colorUseRule=UseColor_Specified, specifiedColor=<rgba 1 1 1 1>}
      uVRegion {bIsValid=false, max=<0 0>, min=<0 0>}
     }
    }
   }
   slot {class=HorizontalBoxSlot, properties {verticalAlignment=VAlign_Center, padding=<ltrb 10 0 0 0>}}
  }
  TutorialHighlight {widgetClassPath=&TutorialHighlight}
 ]
}
