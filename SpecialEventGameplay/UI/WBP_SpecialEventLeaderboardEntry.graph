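# Special event leaderboard row widget showing a player's position, name and score over a rounded background.
# Is Even Row and Is Odd Row turn player index parity into visibility values for alternating backgrounds, returning Collapsed for the local player.
# A red Background_Me starts collapsed, with no logic here to show it.
# The event graph is empty, and nothing here fills the text blocks or calls those functions.
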WidgetBlueprint /SpecialEventGameplay/UI/WBP_SpecialEventLeaderboardEntry
graphs [
 EventGraph {nodes [
   Event PreConstruct()
   Event Construct()
   Event Tick()
  ]}
 Is Even Row {nodes [
   n1 = Is Even Row!() -> (then, Player Index:int, IsLocalPlayer:bool=~) {is=title K2Node_FunctionEntry}
   > n2 = Branch(Condition=@n1.IsLocalPlayer)
   > Return Node!(execute, Result:ESlateVisibility Enum=Collapsed)
   n3 = Branch(execute=@n2.else, Condition="Equal (Integer)"(A="% (Integer)"(A=@n1, B=2) -> (ReturnValue=0)))
   > Return Node(Result=SelfHitTestInvisible)
   Return Node(execute=@n3.else, Result=Collapsed)
  ]}
 Is Odd Row {nodes [
   n1 = Is Odd Row!() -> (then, Player Index:int, Is Local Player:bool=~) {is=title K2Node_FunctionEntry}
   > n2 = Branch(Condition=@n1.Is Local Player)
   > Return Node(Result=Collapsed)
   n3 = Branch(execute=@n2.else, Condition="Equal (Integer)"(A="% (Integer)"(A=@n1, B=2) -> (ReturnValue=0)))
   > Return Node(Result=Collapsed)
   Return Node(execute=@n3.else, Result=SelfHitTestInvisible)
  ]}
]
variables []
widgets {
 info {inheritedWidgetCount=0, parentClass=&UserWidget, rootWidgetClass=&Overlay, widgetCount=15}
 widgets [
  Overlay_0 {widgetClassPath=&Overlay, widget=Overlay_0}
  O_Padding {
   widgetClassPath=&Overlay
   parent=Overlay_0
   widget=O_Padding
   slot {class=OverlaySlot, properties {horizontalAlignment=HAlign_Fill, verticalAlignment=VAlign_Fill, padding=<ltrb 0 2 0 2>}}
  }
  Background_Me {
   widgetClassPath=&Border
   parent=O_Padding
   widget=Background_Me
   properties {
    visibility=Collapsed
    background {
     bIsDynamicallyLoaded=false
     drawAs=Image
     imageType=FullColor
     mirroring=NoMirror
     resourceName=None
     resourceObject=null
     tiling=NoTile
     imageSize=<32 32>
     margin=<ltrb 0 0 0 0>
     outlineSettings {
      bUseBrushTransparency=false
      roundingType=HalfHeightRadius
      width=0
      color {colorUseRule=UseColor_Specified, specifiedColor=<rgba 0 0 0 0>}
      cornerRadii=<0 0 0 0>
     }
     tintColor {colorUseRule=UseColor_Specified, specifiedColor=<rgba 0.4452 0.0296 0.0561 1>}
     uVRegion {bIsValid=false, max=<0 0>, min=<0 0>}
    }
    brushColor=<rgba 1 0.9637 0.9436 1>
   }
   slot {class=OverlaySlot, properties {horizontalAlignment=HAlign_Fill, verticalAlignment=VAlign_Fill}}
  }
  Background_Even {
   widgetClassPath=&Border
   parent=O_Padding
   widget=Background_Even
   properties {
    background {
     bIsDynamicallyLoaded=false
     drawAs=Image
     imageType=FullColor
     mirroring=NoMirror
     resourceName=None
     resourceObject=null
     tiling=NoTile
     imageSize=<32 32>
     margin=<ltrb 0 0 0 0>
     outlineSettings {
      bUseBrushTransparency=false
      roundingType=HalfHeightRadius
      width=0
      color {colorUseRule=UseColor_Specified, specifiedColor=<rgba 0 0 0 0>}
      cornerRadii=<0 0 0 0>
     }
     tintColor {colorUseRule=UseColor_Specified, specifiedColor=<rgba 0.0513 0.0578 0.0685 1>}
     uVRegion {bIsValid=false, max=<0 0>, min=<0 0>}
    }
    brushColor=<rgba 1 0.9637 0.9436 1>
   }
   slot {class=OverlaySlot, properties {horizontalAlignment=HAlign_Fill, verticalAlignment=VAlign_Fill}}
  }
  Background_Odd {
   widgetClassPath=&Border
   parent=O_Padding
   widget=Background_Odd
   properties {
    background {
     bIsDynamicallyLoaded=false
     drawAs=Image
     imageType=FullColor
     mirroring=NoMirror
     resourceName=None
     resourceObject=null
     tiling=NoTile
     imageSize=<32 32>
     margin=<ltrb 0 0 0 0>
     outlineSettings {
      bUseBrushTransparency=false
      roundingType=HalfHeightRadius
      width=0
      color {colorUseRule=UseColor_Specified, specifiedColor=<rgba 0 0 0 0>}
      cornerRadii=<0 0 0 0>
     }
     tintColor {colorUseRule=UseColor_Specified, specifiedColor=<rgba 0.0116 0.0123 0.013 1>}
     uVRegion {bIsValid=false, max=<0 0>, min=<0 0>}
    }
    brushColor=<rgba 1 0.9637 0.9436 1>
   }
   slot {class=OverlaySlot, properties {horizontalAlignment=HAlign_Fill, verticalAlignment=VAlign_Fill}}
  }
  HorizontalBox_1 {
   widgetClassPath=&HorizontalBox
   parent=O_Padding
   widget=HorizontalBox_1
   slot {class=OverlaySlot, properties {horizontalAlignment=HAlign_Fill, verticalAlignment=VAlign_Fill}}
  }
  Spacer {widgetClassPath=&Spacer, parent=HorizontalBox_1, widget=Spacer, properties {size=<16 0>}, slot {class=HorizontalBoxSlot}}
  SizeBox_1 {
   widgetClassPath=&SizeBox
   parent=HorizontalBox_1
   widget=SizeBox_1
   properties {bOverride_MinDesiredWidth=true, minDesiredWidth=80, widthOverride=80}
   slot {class=HorizontalBoxSlot, properties {verticalAlignment=VAlign_Center}}
  }
  PlayerPositionTextBlock {
   widgetClassPath=&CommonTextBlock
   bIsVariable=true
   parent=SizeBox_1
   widget=PlayerPositionTextBlock
   properties {
    text=1
    colorAndOpacity {colorUseRule=UseColor_Specified, specifiedColor=<rgba 0.75 0.75 0.75 1>}
    font {
     bForceMonospaced=true
     bMaterialIsStencil=false
     fontMaterial=null
     fontObject=&BurbankSmall
     letterSpacing=0
     monospacedWidth=0.9
     size=18
     skewAmount=0
     typefaceFontName=Bold
     outlineSettings {
      bApplyOutlineToDropShadows=false
      bMiteredCorners=false
      bSeparateFillAlpha=false
      outlineMaterial=null
      outlineSize=0
      outlineColor=<rgba 0 0 0 1>
     }
    }
    shadowOffset=<0 0>
   }
   slot {class=SizeBoxSlot, properties {horizontalAlignment=HAlign_Right}}
  }
  Spacer_99 {widgetClassPath=&Spacer, parent=HorizontalBox_1, widget=Spacer_99, properties {size=<30 0>}, slot {class=HorizontalBoxSlot}}
  HorizontalBox_0 {
   widgetClassPath=&HorizontalBox
   parent=HorizontalBox_1
   widget=HorizontalBox_0
   slot {class=HorizontalBoxSlot, properties {size {sizeRule=Fill, value=1}}}
  }
  PlayerNameTextBlock {
   widgetClassPath=&CommonTextBlock
   bIsVariable=true
   parent=HorizontalBox_0
   widget=PlayerNameTextBlock
   properties {
    scrollStyle=&TextScrollStyle-Base_C
    text="Player Name ##1132"
    wrappingPolicy=AllowPerCharacterWrapping
    colorAndOpacity {colorUseRule=UseColor_Specified, specifiedColor=<rgba 0.75 0.75 0.75 1>}
    font {
     bForceMonospaced=false
     bMaterialIsStencil=false
     fontMaterial=null
     fontObject=&BurbankSmall
     letterSpacing=0
     monospacedWidth=1
     size=18
     skewAmount=0
     typefaceFontName=Bold
     outlineSettings {
      bApplyOutlineToDropShadows=false
      bMiteredCorners=false
      bSeparateFillAlpha=false
      outlineMaterial=null
      outlineSize=0
      outlineColor=<rgba 0 0 0 1>
     }
    }
    shadowOffset=<0 0>
   }
   slot {class=HorizontalBoxSlot, properties {verticalAlignment=VAlign_Center, size {sizeRule=Fill, value=1}}}
  }
  Spacer_2 {widgetClassPath=&Spacer, parent=HorizontalBox_0, widget=Spacer_2, properties {size=<30 0>}, slot {class=HorizontalBoxSlot}}
  PlayerScoreTextBlock {
   widgetClassPath=&CommonTextBlock
   bIsVariable=true
   parent=HorizontalBox_0
   widget=PlayerScoreTextBlock
   properties {
    justification=Right
    text=218 000 0000
    colorAndOpacity {colorUseRule=UseColor_Specified, specifiedColor=<rgba 0.75 0.75 0.75 1>}
    font {
     bForceMonospaced=true
     bMaterialIsStencil=false
     fontMaterial=null
     fontObject=&BurbankSmall
     letterSpacing=0
     monospacedWidth=0.9
     size=18
     skewAmount=0
     typefaceFontName=Bold
     outlineSettings {
      bApplyOutlineToDropShadows=false
      bMiteredCorners=false
      bSeparateFillAlpha=false
      outlineMaterial=null
      outlineSize=0
      outlineColor=<rgba 0 0 0 1>
     }
    }
    shadowOffset=<0 0>
   }
   slot {
    class=HorizontalBoxSlot
    properties {horizontalAlignment=HAlign_Right, verticalAlignment=VAlign_Center, size {sizeRule=Automatic, value=0.3}}
   }
  }
  Spacer_1 {
   widgetClassPath=&Spacer
   parent=HorizontalBox_1
   widget=Spacer_1
   properties {size=<16 0>}
   slot {class=HorizontalBoxSlot, properties {horizontalAlignment=HAlign_Right}}
  }
 ]
}
