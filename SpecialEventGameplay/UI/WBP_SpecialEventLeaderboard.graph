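# Special event leaderboard screen, a CommonActivatableWidget showing Local and Global leaderboards side by side.
# Each column has a title above a non-selectable CommonListView of WBP_SpecialEventLeaderboardEntry rows, inside a down-only ScaleBox that fits height.
# On Construct it collapses both lists' scrollbars, then hides itself via render opacity and reappears after a short delay.
# The graph contains no logic to populate either list, the global one being named GalacticLeaderboardListView.
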WidgetBlueprint /SpecialEventGameplay/UI/WBP_SpecialEventLeaderboard
graphs [EventGraph {nodes [
   Event Construct()
   > n1 = Sequence!(execute) -> (then_0, then_1, then_2:Exec)
   > SetScrollbarVisibility!(
    execute
    self:List View Base Object Reference=Get LocalLeaderboardListView() -> (LocalLeaderboardListView:Common List View Object Reference)
    InVisibility:ESlateVisibility Enum=Collapsed
   ) -> (then)
   > SetScrollbarVisibility(self=Get GalacticLeaderboardListView() -> (GalacticLeaderboardListView:Common List View Object Reference), InVisibility=Collapsed)
   Delay(execute=@n1.then_2, Duration=0.1)
   > SetRenderOpacity(InOpacity=1)
   SetRenderOpacity(execute=@n1.then_1, InOpacity=0)
  ]}]
variables []
widgets {
 info {inheritedWidgetCount=0, parentClass=&CommonActivatableWidget, rootWidgetClass=&GridPanel, widgetCount=14}
 widgets [
  GridPanel_0 {widgetClassPath=&GridPanel, widget=GridPanel_0, properties {columnFill [1], rowFill [1]}}
  SizeBox_2 {
   widgetClassPath=&SizeBox
   parent=GridPanel_0
   widget=SizeBox_2
   properties {minDesiredWidth=1500, widthOverride=1500}
   slot {class=GridSlot, properties {horizontalAlignment=HAlign_Center}}
  }
  VerticalBox_0 {widgetClassPath=&VerticalBox, parent=SizeBox_2, widget=VerticalBox_0, slot {class=SizeBoxSlot}}
  HB_Body {
   widgetClassPath=&HorizontalBox
   parent=VerticalBox_0
   widget=HB_Body
   slot {class=VerticalBoxSlot, properties {padding=<ltrb 30 0 30 50>, size {sizeRule=Fill, value=1}}}
  }
  VB_Local_Leaderboard {
   widgetClassPath=&VerticalBox
   parent=HB_Body
   widget=VB_Local_Leaderboard
   slot {class=HorizontalBoxSlot, properties {size {sizeRule=Fill, value=1}}}
  }
  CommonBorder_0 {
   widgetClassPath=&CommonBorder
   parent=VB_Local_Leaderboard
   widget=CommonBorder_0
   properties {
    background {
     bIsDynamicallyLoaded=false
     drawAs=Image
     imageType=FullColor
     mirroring=NoMirror
     resourceName=None
     resourceObject=null
     tiling=NoTile
     imageSize=<32 32>
     margin=<ltrb 0 0 0 0>
     outlineSettings {
      bUseBrushTransparency=false
      roundingType=HalfHeightRadius
      width=0
      color {colorUseRule=UseColor_Specified, specifiedColor=<rgba 0 0 0 0>}
      cornerRadii=<0 0 0 0>
     }
     tintColor {colorUseRule=UseColor_Specified, specifiedColor=<rgba 1 1 1 0>}
     uVRegion {bIsValid=false, max=<0 0>, min=<0 0>}
    }
    padding=<ltrb 60 24 60 24>
   }
   slot {class=VerticalBoxSlot, properties {padding=<ltrb 0 0 0 20>, size {sizeRule=Automatic, value=0.1}}}
  }
  CommonTextBlock_1 {
   widgetClassPath=&CommonTextBlock
   parent=CommonBorder_0
   widget=CommonTextBlock_1
   properties {
    autoWrapText=true
    justification=Center
    text=LOCAL LEADERBOARD
    font {
     bForceMonospaced=false
     bMaterialIsStencil=false
     fontMaterial=null
     fontObject=&BurbankSmall
     letterSpacing=0
     monospacedWidth=1
     size=27
     skewAmount=0
     typefaceFontName=Bold
     outlineSettings {
      bApplyOutlineToDropShadows=false
      bMiteredCorners=false
      bSeparateFillAlpha=false
      outlineMaterial=null
      outlineSize=0
      outlineColor=<rgba 0 0 0 1>
     }
    }
    shadowOffset=<0 0>
   }
   slot {class=BorderSlot, properties {padding=<ltrb 60 24 60 24>}}
  }
  ScaleBox_0 {
   widgetClassPath=&ScaleBox
   parent=VB_Local_Leaderboard
   widget=ScaleBox_0
   properties {stretch=ScaleToFitY, stretchDirection=DownOnly}
   slot {class=VerticalBoxSlot, properties {size {sizeRule=Fill, value=1}}}
  }
  LocalLeaderboardListView {
   widgetClassPath=&CommonListView
   bIsVariable=true
   parent=ScaleBox_0
   widget=LocalLeaderboardListView
   properties {
    entryWidgetClass=&/WBP_SpecialEventLeaderboardEntry.WBP_SpecialEventLeaderboardEntry_C
    numDesignerPreviewEntries=20
    selectionMode=None
   }
   slot {class=ScaleBoxSlot, properties {horizontalAlignment=HAlign_Fill, verticalAlignment=VAlign_Fill}}
  }
  VB_Galactic_Leaderboard {
   widgetClassPath=&VerticalBox
   parent=HB_Body
   widget=VB_Galactic_Leaderboard
   slot {class=HorizontalBoxSlot, properties {padding=<ltrb 100 0 0 0>, size {sizeRule=Fill, value=1}}}
  }
  CommonBorder {
   widgetClassPath=&CommonBorder
   parent=VB_Galactic_Leaderboard
   widget=CommonBorder
   properties {
    background {
     bIsDynamicallyLoaded=false
     drawAs=Image
     imageType=FullColor
     mirroring=NoMirror
     resourceName=None
     resourceObject=null
     tiling=NoTile
     imageSize=<32 32>
     margin=<ltrb 0 0 0 0>
     outlineSettings {
      bUseBrushTransparency=false
      roundingType=HalfHeightRadius
      width=0
      color {colorUseRule=UseColor_Specified, specifiedColor=<rgba 0 0 0 0>}
      cornerRadii=<0 0 0 0>
     }
     tintColor {colorUseRule=UseColor_Specified, specifiedColor=<rgba 1 1 1 0>}
     uVRegion {bIsValid=false, max=<0 0>, min=<0 0>}
    }
    padding=<ltrb 60 24 60 24>
   }
   slot {class=VerticalBoxSlot, properties {padding=<ltrb 0 0 0 20>, size {sizeRule=Automatic, value=0.1}}}
  }
  CommonTextBlock {
   widgetClassPath=&CommonTextBlock
   parent=CommonBorder
   widget=CommonTextBlock
   properties {
    justification=Center
    text=GLOBAL LEADERBOARD
    font {
     bForceMonospaced=false
     bMaterialIsStencil=false
     fontMaterial=null
     fontObject=&BurbankSmall
     letterSpacing=0
     monospacedWidth=1
     size=27
     skewAmount=0
     typefaceFontName=Bold
     outlineSettings {
      bApplyOutlineToDropShadows=false
      bMiteredCorners=false
      bSeparateFillAlpha=false
      outlineMaterial=null
      outlineSize=0
      outlineColor=<rgba 0 0 0 1>
     }
    }
    shadowOffset=<0 0>
   }
   slot {class=BorderSlot, properties {padding=<ltrb 60 24 60 24>}}
  }
  ScaleBox {
   widgetClassPath=&ScaleBox
   parent=VB_Galactic_Leaderboard
   widget=ScaleBox
   properties {stretch=ScaleToFitY, stretchDirection=DownOnly}
   slot {class=VerticalBoxSlot, properties {size {sizeRule=Fill, value=1}}}
  }
  GalacticLeaderboardListView {
   widgetClassPath=&CommonListView
   bIsVariable=true
   parent=ScaleBox
   widget=GalacticLeaderboardListView
   properties {
    entryWidgetClass=&WBP_SpecialEventLeaderboardEntry_C
    numDesignerPreviewEntries=20
    selectionMode=None
   }
   slot {class=ScaleBoxSlot, properties {horizontalAlignment=HAlign_Fill, verticalAlignment=VAlign_Fill}}
  }
 ]
}
